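; Drill crater terrain Blueprint for the Artemis HollowBase environment that configures a runtime virtual texture decal from exposed variables.
; The construction script creates a dynamic instance of the shared M_RVT_Decal_Blueprint material on the RVT_Decal component and assigns normal, detail normal, specular mask and diffuse textures.
; It then pushes two colors and scalar controls for scale, masking, roughness, specular, detail and blending.
; BeginPlay, overlap and tick events are empty.
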(event EventBeginPlay)

(event Collision|EventActorBeginOverlap
  (params (OtherActor :type Actor%20Object%20Reference)))

(event EventTick
  (params (DeltaSeconds :type Float%20%28single-precision%29)))

(fn ConstructionScript
  (bind _rvt_decal (Variables|Default|GetRVT_Decal))
  (bind _returnvalue (Rendering|Material|CreateDynamicMaterialInstance _rvt_decal 0 "/Game/Athena/WinterArt/Materials/M_RVT_Decal_Blueprint.M_RVT_Decal_Blueprint" "None" :!class /Script/Engine.PrimitiveComponent))
  (Rendering|Material|SetTextureParameterValue _returnvalue "Normal" (Variables|Texture|GetNormal))
  (Rendering|Material|SetTextureParameterValue _returnvalue "DetailsNormal" (Variables|Texture|GetDetailsNormal))
  (Rendering|Material|SetTextureParameterValue _returnvalue "Specular" (Variables|Texture|GetSpecMask))
  (Rendering|Material|SetTextureParameterValue _returnvalue "Diffuse" (Variables|Texture|GetDiffuse))
  (Rendering|Material|SetVectorParameterValueonMaterials _rvt_decal "DecalColor" (Math|Conversions|ToVector(LinearColor) (Variables|Colors|GetColorA)))
  (Rendering|Material|SetVectorParameterValueonMaterials _rvt_decal "DecalColor2" (Math|Conversions|ToVector(LinearColor) (Variables|Colors|GetColor%20B)))
  (Rendering|Material|SetScalarParameterValueonMaterials _rvt_decal "ScaleTextures" (Variables|Colors|GetScaleTextures))
  (Rendering|Material|SetScalarParameterValueonMaterials _rvt_decal "MaskDensity" (Variables|Colors|GetMaskDensity))
  (Rendering|Material|SetScalarParameterValueonMaterials _rvt_decal "RoughnessTweak" (Variables|Colors|GetRoughnessTweak))
  (Rendering|Material|SetScalarParameterValueonMaterials _rvt_decal "SpecularTweak" (Variables|Colors|GetSpecularTweak))
  (Rendering|Material|SetScalarParameterValueonMaterials _rvt_decal "ColorTransition " (Variables|Colors|GetColorTransition%20))
  (Rendering|Material|SetScalarParameterValueonMaterials _rvt_decal "ColorContrast" (Variables|Colors|GetColorContrast))
  (Rendering|Material|SetScalarParameterValueonMaterials _rvt_decal "DetailsValue" (Variables|Colors|GetDetailsValue))
  (Rendering|Material|SetScalarParameterValueonMaterials _rvt_decal "DetailsScale" (Variables|Colors|GetDetailsScale))
  (Rendering|Material|SetScalarParameterValueonMaterials _rvt_decal "RoughnessTweak2" (Variables|Colors|GetRoughnessTweakB))
  (Rendering|Material|SetScalarParameterValueonMaterials _rvt_decal "SpecularTweak2" (Variables|Colors|GetSpecularTweakB))
  (Rendering|Material|SetScalarParameterValueonMaterials _rvt_decal "Blendin" (Variables|Colors|GetBlend%20In)))
